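; Behavior tree evaluator for the Umbramolt Fraternal AI bot whose starting condition passes while the twin is inactive.
; On enter, it caches the owning Athena bot controller and its controlled player pawn as the Fraternal controller and pawn.
; The condition calls the parent, ignores its result, and returns the negated GetTwinState interface result from the cached pawn.
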(event EventOnEnter
  (params (OwnerComp :type Behavior%20Tree%20Component%20Object%20Reference))
  (bind _asfort_athena_aibot_controller (Utilities|Casting|CastToFortAthenaAIBotController (Components|GetOwner OwnerComp)))
  (bind _output_get (Variables|Default|SetFraternalController _asfort_athena_aibot_controller))
  (bind _asfort_player_pawn_athena (Utilities|Casting|CastToFortPlayerPawnAthena (Pawn|GetControlledPawn _output_get)))
  (Variables|Default|SetFraternalPawn _asfort_player_pawn_athena))

(fn EvaluateStartingConditions
  (params (OwnerComp :type Behavior%20Tree%20Component%20Object%20Reference))
  (|CallParentFunction OwnerComp)
  (bind _isactive (Class|UmbraMoltFraternalInterface|GetTwinState(Message) (Variables|Default|GetFraternalPawn)))
  (return (not _isactive)))
