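; Behavior tree injection evaluator that gates the walkout behavior of the Umbramolt Fraternal AI bot.
; On enter, it casts the owning controller to an Athena AI bot controller and caches it along with its controlled Athena player pawn.
; Starting conditions pass when that pawn has the StartWalkout gameplay tag and lacks WalkoutComplete.
; The parent condition check is called, but its result is not used in the returned value.
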(event EventOnEnter
  (params (OwnerComp :type Behavior%20Tree%20Component%20Object%20Reference))
  (bind _asfort_athena_aibot_controller (Utilities|Casting|CastToFortAthenaAIBotController (Components|GetOwner OwnerComp)))
  (bind _output_get (Variables|Default|SetFraternalController _asfort_athena_aibot_controller))
  (bind _asfort_player_pawn_athena (Utilities|Casting|CastToFortPlayerPawnAthena (Pawn|GetControlledPawn _output_get)))
  (Variables|Default|SetFraternalPawn _asfort_player_pawn_athena))

(fn EvaluateStartingConditions
  (params (OwnerComp :type Behavior%20Tree%20Component%20Object%20Reference))
  (|CallParentFunction OwnerComp)
  (return (and (GameplayTags|HasMatchingGameplayTag (Variables|Default|GetFraternalPawn) "(TagName=\"Umbramolt_Fraternal.StartWalkout\")") (not (GameplayTags|HasMatchingGameplayTag (Variables|Default|GetFraternalPawn) "(TagName=\"Umbramolt_Fraternal.WalkoutComplete\")")))))
